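# Sound cue for dusk crow "crazy" calls, an ambient one-shot.
# It picks one of three crow waves at equal weight, then a modulator randomizes pitch slightly around normal.
# It uses tile-ambient one-shot attenuation and an overridden concurrency limit that stops the quietest instance when too many play.
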SoundCue /Hermes/Audio/Crows/Dusk_Crows_Crazy_Cue
root=Modulator_0
nodes [
 {id=Modulator_0, children [Random_1], properties {pitchMax=1.2, pitchMin=0.9}}
 {id=Random_1, children [WavePlayer_3,WavePlayer_4,WavePlayer_5], properties {weights [1,1,1]}}
 {id=WavePlayer_3, properties {soundWaveAssetPtr=&/crow_crazy_09}}
 {id=WavePlayer_4, properties {soundWaveAssetPtr=&/crow_crazy_02}}
 {id=WavePlayer_5, properties {soundWaveAssetPtr=&/crow_crazy_01}}
]
properties {
 attenuationSettings=/Game/Sounds/Attenuation/Fort_TileAmbient_OneShot_Att
 bOverrideConcurrency=true
 maxDistance=8000
 soundClassObject=&Fort_Ambient_OneShots
 concurrencyOverrides {maxCount=6, resolutionRule=StopQuietest}
}
