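# Sound cue for outdoor ambient crow one-shots, spatialized by a bird attenuation asset with its own concurrency limit and a reverb submix send.
# All quality levels share one random node that picks, at equal weight, one of three crow groups or an empty slot.
# Two groups randomly choose a closer or Eastern Europe call, while the third layers three rare mountain calls simultaneously.
# Each group passes through its own pitch/volume modulator and mixer.
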SoundCue /Hermes/Audio/Crows/Spot_Bird_Crow_OneShot_Cue
root=QualityLevel_0
nodes [
 {id=Mixer_0, children [Modulator_0], properties {inputVolume [0.8]}}
 {id=Mixer_1, children [Modulator_1], properties {inputVolume [0.8]}}
 {id=Mixer_2, children [Modulator_2], properties {inputVolume [0.8]}}
 {id=Mixer_3, children [WavePlayer_17, WavePlayer_18, WavePlayer_19], properties {inputVolume [1,1,1]}}
 {id=Modulator_0, children [Mixer_3], properties {pitchMin=0.8, volumeMax=1, volumeMin=0.8}}
 {id=Modulator_1, children [Random_3]}
 {id=Modulator_2, children [Random_2]}
 {id=QualityLevel_0, children [Random_5,Random_5,Random_5,Random_5]}
 {id=Random_2, children [WavePlayer_5, WavePlayer_9, WavePlayer_11, WavePlayer_12], properties {weights [1,1,1,1]}}
 {id=Random_3, children [WavePlayer_13, WavePlayer_14, WavePlayer_15, WavePlayer_16], properties {weights [1,1,1,1]}}
 {id=Random_5, children [Mixer_0,Mixer_1,Mixer_2,null], properties {weights [1,1,1,1]}}
 {id=WavePlayer_11, properties {soundWaveAssetPtr=&/Spot_Bird_Crow_Closer_02}}
 {id=WavePlayer_12, properties {soundWaveAssetPtr=&/Spot_Bird_Crow_Closer_01}}
 {id=WavePlayer_13, properties {soundWaveAssetPtr=&/Spot_Bird_Crow_DayEvening_Eastern_Europe_01}}
 {id=WavePlayer_14, properties {soundWaveAssetPtr=&/Spot_Bird_Crow_DayEvening_Eastern_Europe_02}}
 {id=WavePlayer_15, properties {soundWaveAssetPtr=&/Spot_Bird_Crow_DayEvening_Eastern_Europe_03}}
 {id=WavePlayer_16, properties {soundWaveAssetPtr=&/Spot_Bird_Crow_DayEvening_Eastern_Europe_04}}
 {id=WavePlayer_17, properties {soundWaveAssetPtr=&/Spot_Bird_Rare_Mountain_Crow_03}}
 {id=WavePlayer_18, properties {soundWaveAssetPtr=&/Spot_Bird_Rare_Mountain_Crow_02}}
 {id=WavePlayer_19, properties {soundWaveAssetPtr=&/Spot_Bird_Rare_Mountain_Crow_01}}
 {id=WavePlayer_5, properties {soundWaveAssetPtr=&/Spot_Bird_Crow_Closer_04}}
 {id=WavePlayer_9, properties {soundWaveAssetPtr=&/Spot_Bird_Crow_Closer_03}}
]
properties {
 attenuationSettings=/AmbientAudioBase/Attenuations/Fort_TileAmbient_OneShot_Bird_Att
 bEnableBusSends=false
 bOverrideConcurrency=true
 maxDistance=8005
 soundClassObject=&Fort_Ambient_OneShots_Outdoors
 volumeMultiplier=1
 concurrencyOverrides {maxCount=6, voiceStealReleaseTime=0.25}
 preEffectBusSends [{}]
 soundSubmixSends [{maxSendLevel=1.2, minSendLevel=0.4, sendLevel=0, soundSubmix=/Engine/EngineSounds/Submixes/MasterReverbSubmixDefault}]
}
